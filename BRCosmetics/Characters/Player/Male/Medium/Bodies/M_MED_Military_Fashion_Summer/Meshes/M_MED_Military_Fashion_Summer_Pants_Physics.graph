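# Physics asset for the pants of the male medium Military Fashion Summer body, defining dynamic front and back pants pieces on each leg.
# Capsules cover the pelvis and thighs, with each thigh constrained to the pelvis.
# Each thigh carries tiny front and back base capsules on angularly free joints.
# Each base drives an offset box body through a tighter limited joint, forming short two-link chains for cloth-like motion.
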PhysicsAsset /BRCosmetics/Characters/Player/Male/Medium/Bodies/M_MED_Military_Fashion_Summer/Meshes/M_MED_Military_Fashion_Summer_Pants_Physics
bodies {
 dyn_pants_bk_base_l [{shapeType=Capsule, length=1, radius=1}]
 dyn_pants_bk_base_r [{shapeType=Capsule, length=1, radius=1}]
 dyn_pants_bk_l [{shapeType=Box, extentX=9.5, extentY=2.5, extentZ=10, center=<4 0 0>}]
 dyn_pants_bk_r [{shapeType=Box, extentX=9.5, extentY=2.5, extentZ=10, center=<-4 0 0>}]
 dyn_pants_fr_base_l [{shapeType=Capsule, length=1, radius=1}]
 dyn_pants_fr_base_r [{shapeType=Capsule, length=1, radius=1}]
 dyn_pants_fr_l [{shapeType=Box, extentX=9.5, extentY=2.5, extentZ=10, center=<4 0 0>}]
 dyn_pants_fr_r [{shapeType=Box, extentX=9.5, extentY=2.5, extentZ=10, center=<-4 0 0>}]
 pelvis [{shapeType=Capsule, length=8, radius=10.75, center=<0 -0.582 0>}]
 thigh_l [{shapeType=Capsule, length=23.25, radius=8.75, center=<-18.9426 0.3464 0.9332>, rotation=<rot 90 0 0>}]
 thigh_r [{shapeType=Capsule, length=23.25, radius=8.75, center=<18.9426 -0.3464 -0.9332>, rotation=<rot 90 0 180>}]
}
constraints [
 {bone1Name=thigh_l, bone2Name=pelvis, swing1LimitDegrees=45, swing2LimitDegrees=45, twistLimitDegrees=45}
 {bone1Name=thigh_r, bone2Name=pelvis, swing1LimitDegrees=45, swing2LimitDegrees=45, twistLimitDegrees=45}
 {
  bone1Name=dyn_pants_bk_base_r
  bone2Name=thigh_r
  swing1LimitDegrees=45
  swing1Motion=Free
  swing2LimitDegrees=45
  swing2Motion=Free
  twistLimitDegrees=45
  twistMotion=Free
 }
 {
  bone1Name=dyn_pants_fr_base_r
  bone2Name=thigh_r
  swing1LimitDegrees=45
  swing1Motion=Free
  swing2LimitDegrees=45
  swing2Motion=Free
  twistLimitDegrees=45
  twistMotion=Free
 }
 {bone1Name=dyn_pants_bk_r, bone2Name=dyn_pants_bk_base_r, swing1LimitDegrees=30, swing2LimitDegrees=20, twistLimitDegrees=10}
 {bone1Name=dyn_pants_fr_r, bone2Name=dyn_pants_fr_base_r, swing1LimitDegrees=30, swing2LimitDegrees=20, twistLimitDegrees=10}
 {bone1Name=dyn_pants_fr_l, bone2Name=dyn_pants_fr_base_l, swing1LimitDegrees=30, swing2LimitDegrees=20, twistLimitDegrees=10}
 {bone1Name=dyn_pants_bk_l, bone2Name=dyn_pants_bk_base_l, swing1LimitDegrees=30, swing2LimitDegrees=20, twistLimitDegrees=10}
 {
  bone1Name=dyn_pants_fr_base_l
  bone2Name=thigh_l
  swing1LimitDegrees=45
  swing1Motion=Free
  swing2LimitDegrees=45
  swing2Motion=Free
  twistLimitDegrees=45
  twistMotion=Free
 }
 {
  bone1Name=dyn_pants_bk_base_l
  bone2Name=thigh_l
  swing1LimitDegrees=45
  swing1Motion=Free
  swing2LimitDegrees=45
  swing2Motion=Free
  twistLimitDegrees=45
  twistMotion=Free
 }
]
