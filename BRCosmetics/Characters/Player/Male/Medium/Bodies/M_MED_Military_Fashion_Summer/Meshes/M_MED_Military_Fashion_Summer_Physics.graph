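# Physics asset for the male medium Military Fashion Summer body cosmetic, defining physics for a hanging pouch with straps.
# Capsules on the pelvis and both thighs form the body shapes around the pouch.
# The pouch base rotates freely on the pelvis and carries two box segments in a chain, each locked to a single limited swing axis like a hinge.
# Two strap chains of two boxes each branch from the end pouch segment with limited swing and tight twist.
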PhysicsAsset /BRCosmetics/Characters/Player/Male/Medium/Bodies/M_MED_Military_Fashion_Summer/Meshes/M_MED_Military_Fashion_Summer_Physics
bodies {
 dyn_pouch_a [{shapeType=Box, extentX=8.01, extentY=2.51, extentZ=2.51, center=<2.104 0 0>, rotation=<rot 0 -90 0>}]
 dyn_pouch_a_base [{shapeType=Capsule, length=0.505, radius=0.505, rotation=<rot 0 -90 0>}]
 dyn_pouch_b [{shapeType=Box, extentX=14.0797, extentY=6.5014, extentZ=17.4709, center=<9.0898 -0.3236 -1.1352>, rotation=<rot 0 -90.869 -90.1252>}]
 dyn_pouch_strap_aa [{shapeType=Box, extentX=5, extentY=3, extentZ=1.5, center=<2.5 0 0>}]
 dyn_pouch_strap_ab [{shapeType=Box, extentX=5, extentY=3, extentZ=1.5, center=<2.5 0 0>}]
 dyn_pouch_strap_ba [{shapeType=Box, extentX=5, extentY=3, extentZ=1.5, center=<2.5 0 0>}]
 dyn_pouch_strap_bb [{shapeType=Box, extentX=5, extentY=3, extentZ=1.5, center=<2.5 0 0>}]
 pelvis [{shapeType=Capsule, length=8, radius=10.75, center=<0 -0.582 0>}]
 thigh_l [{shapeType=Capsule, length=23.25, radius=8.75, center=<-18.9426 0.3464 0.9332>, rotation=<rot 90 0 0>}]
 thigh_r [{shapeType=Capsule, length=23.25, radius=8.75, center=<18.9426 -0.3464 -0.9332>, rotation=<rot 90 0 180>}]
}
constraints [
 {bone1Name=thigh_l, bone2Name=pelvis, swing1LimitDegrees=45, swing2LimitDegrees=45, twistLimitDegrees=45}
 {bone1Name=thigh_r, bone2Name=pelvis, swing1LimitDegrees=45, swing2LimitDegrees=45, twistLimitDegrees=45}
 {
  bone1Name=dyn_pouch_a_base
  bone2Name=pelvis
  swing1LimitDegrees=45
  swing1Motion=Free
  swing2LimitDegrees=45
  swing2Motion=Free
  twistLimitDegrees=45
  twistMotion=Free
 }
 {
  bone1Name=dyn_pouch_a
  bone2Name=dyn_pouch_a_base
  swing1LimitDegrees=5
  swing1Motion=Locked
  swing2LimitDegrees=5
  twistLimitDegrees=5
  twistMotion=Locked
 }
 {
  bone1Name=dyn_pouch_b
  bone2Name=dyn_pouch_a
  swing1LimitDegrees=5
  swing1Motion=Locked
  swing2LimitDegrees=45
  twistLimitDegrees=5
  twistMotion=Locked
 }
 {bone1Name=dyn_pouch_strap_ba, bone2Name=dyn_pouch_b, swing1LimitDegrees=30, swing2LimitDegrees=30, twistLimitDegrees=5}
 {bone1Name=dyn_pouch_strap_bb, bone2Name=dyn_pouch_strap_ba, swing1LimitDegrees=30, swing2LimitDegrees=30, twistLimitDegrees=5}
 {bone1Name=dyn_pouch_strap_aa, bone2Name=dyn_pouch_b, swing1LimitDegrees=15, swing2LimitDegrees=45, twistLimitDegrees=5}
 {bone1Name=dyn_pouch_strap_ab, bone2Name=dyn_pouch_strap_aa, swing1LimitDegrees=15, swing2LimitDegrees=45, twistLimitDegrees=5}
]
