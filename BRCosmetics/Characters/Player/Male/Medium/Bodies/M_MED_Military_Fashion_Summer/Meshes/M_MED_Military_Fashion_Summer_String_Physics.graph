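# Physics asset for the front left and right dangling straps, or strings, on the male medium Military Fashion Summer player cosmetic body.
# Each side chains a tiny base capsule, angularly free against the pelvis, to a longer strap capsule with limited swing and tight twist.
# Larger capsules on the pelvis and both thighs, linked by limited constraints, approximate the hips and upper legs around the straps.
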PhysicsAsset /BRCosmetics/Characters/Player/Male/Medium/Bodies/M_MED_Military_Fashion_Summer/Meshes/M_MED_Military_Fashion_Summer_String_Physics
bodies {
 dyn_fr_strap_base_l [{shapeType=Capsule, length=1, radius=1}]
 dyn_fr_strap_base_r [{shapeType=Capsule, length=1, radius=1}]
 dyn_fr_strap_l [{shapeType=Capsule, length=7, radius=1.5, center=<4 -0.1316 0>, rotation=<rot 90 0 0>}]
 dyn_fr_strap_r [{shapeType=Capsule, length=7, radius=1.5, center=<-4 0 0.0044>, rotation=<rot 90 0 0>}]
 pelvis [{shapeType=Capsule, length=8.25, radius=10.5, center=<-9.5958 -0.7995 0.0001>}]
 thigh_l [{shapeType=Capsule, length=19.75, radius=9.5, center=<-18.9426 0.3464 1.4054>, rotation=<rot 90 0 0>}]
 thigh_r [{shapeType=Capsule, length=19.75, radius=9.5, center=<18.9426 -0.3463 -1.8896>, rotation=<rot 90 0 -179.9999>}]
}
constraints [
 {bone1Name=thigh_l, bone2Name=pelvis, swing1LimitDegrees=45, swing2LimitDegrees=45, twistLimitDegrees=45}
 {bone1Name=thigh_r, bone2Name=pelvis, swing1LimitDegrees=45, swing2LimitDegrees=45, twistLimitDegrees=45}
 {
  bone1Name=dyn_fr_strap_base_l
  bone2Name=pelvis
  swing1LimitDegrees=45
  swing1Motion=Free
  swing2LimitDegrees=45
  swing2Motion=Free
  twistLimitDegrees=45
  twistMotion=Free
 }
 {
  bone1Name=dyn_fr_strap_base_r
  bone2Name=pelvis
  swing1LimitDegrees=45
  swing1Motion=Free
  swing2LimitDegrees=45
  swing2Motion=Free
  twistLimitDegrees=45
  twistMotion=Free
 }
 {bone1Name=dyn_fr_strap_l, bone2Name=dyn_fr_strap_base_l, swing1LimitDegrees=45, swing2LimitDegrees=30, twistLimitDegrees=5}
 {bone1Name=dyn_fr_strap_r, bone2Name=dyn_fr_strap_base_r, swing1LimitDegrees=45, swing2LimitDegrees=30, twistLimitDegrees=5}
]
