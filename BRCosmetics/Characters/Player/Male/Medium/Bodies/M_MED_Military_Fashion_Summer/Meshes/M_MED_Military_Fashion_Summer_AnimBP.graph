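# Post-process style AnimBlueprint for the male Military Fashion Summer body cosmetic, copying the attached parent's pose and adding secondary motion to thighs, pants, straps and pouch.
# Weighted constraints place helper bones, then thigh and knee yaw, read each update, drives corrective bone offsets.
# Three RigidBody passes simulate dynamics; crouching disables the string pass and reduces pants to 0.3.
# Look-At nodes aim buckle straps, with a final thigh-driven correction.
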AnimBlueprint /BRCosmetics/Characters/Player/Male/Medium/Bodies/M_MED_Military_Fashion_Summer/Meshes/M_MED_Military_Fashion_Summer_AnimBP
target_skeleton=&/M_MED_Military_Fashion_Summer_Skeleton
graphs [
 AnimGraph {nodes [
   n1 = "Constraint - thigh_volume_l (R)"(ComponentPose=Local To Component(LocalPose=Copy Pose From Mesh() {bUseAttachedParent=true}), Alpha=0.1, ConstraintWeights_0=1) {
    alphaBoolBlend {blendOption=Linear}
    boneToModify {boneName=thigh_volume_l}
    constraintSetup [{transformType=Rotation, targetBone {boneName=thigh_l}}]
    constraintWeights [1]
   }
   n2 = "Transform (Modify) Bone - Bone: thigh_volume_l"(ComponentPose=@n1, Alpha=Get uplegRot_L() -> (uplegRot_L:double), Translation="-3,-3,0", Rotation="0,0,0", Scale="1.2,1.2,1.2") {
    rotationMode=BMM_Additive
    rotationSpace=BCS_BoneSpace
    scaleMode=BMM_Additive
    scaleSpace=BCS_BoneSpace
    translationMode=BMM_Additive
    translationSpace=BCS_BoneSpace
    alphaBoolBlend {blendOption=Linear}
    alphaScaleBiasClamp {bMapRange=true, inRange {max=0, min=-100}, outRange {max=0, min=1}}
    boneToModify {boneName=thigh_volume_l}
    scale=<1.2 1.2 1.2>
    translation=<-3 -3>
   }
   n3 = "Constraint - thigh_volume_r (R)"(ComponentPose=@n2, Alpha=0.1, ConstraintWeights_0=1) {
    alphaBoolBlend {blendOption=Linear}
    boneToModify {boneName=thigh_volume_r}
    constraintSetup [{transformType=Rotation, targetBone {boneName=thigh_r}}]
    constraintWeights [1]
   }
   n4 = "Transform (Modify) Bone - Bone: thigh_volume_r"(ComponentPose=@n3, Alpha=Get uplegRot_R() -> (uplegRot_R:double), Translation="3,3,0", Rotation="0,0,0", Scale="1.2,1.2,1.2") {
    rotationMode=BMM_Replace
    rotationSpace=BCS_BoneSpace
    scaleMode=BMM_Additive
    scaleSpace=BCS_BoneSpace
    translationMode=BMM_Additive
    translationSpace=BCS_BoneSpace
    alphaBoolBlend {blendOption=Linear}
    alphaScaleBiasClamp {bMapRange=true, inRange {max=0, min=-100}, outRange {max=0, min=1}}
    boneToModify {boneName=thigh_volume_r}
    scale=<1.2 1.2 1.2>
    translation=<3 3>
   }
   n5 = "Constraint - dyn_pouch_a_base (PPPP)"!(
    ComponentPose=@n4
    Alpha=1
    ConstraintWeights_0=0.2279
    ConstraintWeights_1:float=0.0233
    ConstraintWeights_2:float=0.0373
    ConstraintWeights_3:float=0.7115
   ) -> (Pose) {
    alphaBoolBlend {blendOption=Linear}
    boneToModify {boneName=dyn_pouch_a_base}
    constraintSetup [{targetBone {boneName=spine_01}}, {targetBone {boneName=spine_02}}, {targetBone {boneName=thigh_twist_01_r}}, {targetBone {boneName=pelvis}}]
    constraintWeights [1,1,1,1]
   }
   n6 = "Constraint - dyn_pouch_bukle_strap_a_tgt (PPP)"!(ComponentPose=@n5, Alpha=1, ConstraintWeights_0=0.0973, ConstraintWeights_1:float=0.2067, ConstraintWeights_2:float=0.696) -> (Pose) {
    alphaBoolBlend {blendOption=Linear}
    boneToModify {boneName=dyn_pouch_bukle_strap_a_tgt}
    constraintSetup [{targetBone {boneName=spine_01}}, {targetBone {boneName=thigh_twist_01_r}}, {targetBone {boneName=pelvis}}]
    constraintWeights [1,1,1]
   }
   n7 = "Constraint - dyn_pouch_bukle_strap_b_tgt (PPP)"!(ComponentPose=@n6, Alpha=1, ConstraintWeights_0=0.2549, ConstraintWeights_1:float=0.0352, ConstraintWeights_2:float=0.7099) -> (Pose) {
    alphaBoolBlend {blendOption=Linear}
    boneToModify {boneName=dyn_pouch_bukle_strap_b_tgt}
    constraintSetup [{targetBone {boneName=spine_01}}, {targetBone {boneName=spine_02}}, {targetBone {boneName=pelvis}}]
    constraintWeights [1,1,1]
   }
   n8 = "Constraint - dyn_fr_strap_base_l (PPP)"!(ComponentPose=@n7, Alpha=1, ConstraintWeights_0=0.621, ConstraintWeights_1:float=0.3376, ConstraintWeights_2:float=0.0414) -> (Pose) {
    alphaBoolBlend {blendOption=Linear}
    boneToModify {boneName=dyn_fr_strap_base_l}
    constraintSetup [{targetBone {boneName=pelvis}}, {targetBone {boneName=thigh_twist_01_l}}, {targetBone {boneName=thigh_twist_01_r}}]
    constraintWeights [1,1,1]
   }
   n9 = "Constraint - dyn_fr_strap_base_r (PPP)"!(ComponentPose=@n8, Alpha=1, ConstraintWeights_0=0.5895, ConstraintWeights_1:float=0.0447, ConstraintWeights_2:float=0.3658) -> (Pose) {
    alphaBoolBlend {blendOption=Linear}
    boneToModify {boneName=dyn_fr_strap_base_r}
    constraintSetup [{targetBone {boneName=pelvis}}, {targetBone {boneName=thigh_twist_01_l}}, {targetBone {boneName=thigh_twist_01_r}}]
    constraintWeights [1,1,1]
   }
   n10 = "Constraint - dyn_pants_fr_base_l (PP)"!(ComponentPose=@n9, Alpha=1, ConstraintWeights_0=0.4368, ConstraintWeights_1:float=0.5632) -> (Pose) {
    alphaBoolBlend {blendOption=Linear}
    boneToModify {boneName=dyn_pants_fr_base_l}
    constraintSetup [{targetBone {boneName=thigh_twist_01_l}}, {targetBone {boneName=thigh_l}}]
    constraintWeights [1,1]
   }
   n11 = "Constraint - dyn_pants_bk_base_l (PP)"!(ComponentPose=@n10, Alpha=1, ConstraintWeights_0=0.3304, ConstraintWeights_1:float=0.6696) -> (Pose) {
    alphaBoolBlend {blendOption=Linear}
    boneToModify {boneName=dyn_pants_bk_base_l}
    constraintSetup [{targetBone {boneName=thigh_twist_01_l}}, {targetBone {boneName=thigh_l}}]
    constraintWeights [1,1]
   }
   n12 = "Constraint - dyn_pants_fr_base_r (PP)"!(ComponentPose=@n11, Alpha=1, ConstraintWeights_0=0.5484, ConstraintWeights_1:float=0.4516) -> (Pose) {
    alphaBoolBlend {blendOption=Linear}
    boneToModify {boneName=dyn_pants_fr_base_r}
    constraintSetup [{targetBone {boneName=thigh_r}}, {targetBone {boneName=thigh_twist_01_r}}]
    constraintWeights [1,1]
   }
   n13 = "Constraint - dyn_pants_bk_base_r (PP)"!(ComponentPose=@n12, Alpha=1, ConstraintWeights_0=0.6544, ConstraintWeights_1:float=0.3456) -> (Pose) {
    alphaBoolBlend {blendOption=Linear}
    boneToModify {boneName=dyn_pants_bk_base_r}
    constraintSetup [{targetBone {boneName=thigh_r}}, {targetBone {boneName=thigh_twist_01_r}}]
    constraintWeights [1,1]
   }
   n14 = "Constraint - dyn_fr_strap_base_l (R)"(ComponentPose=@n13, Alpha=1, ConstraintWeights_0=1) {
    alphaBoolBlend {blendOption=Linear}
    boneToModify {boneName=dyn_fr_strap_base_l}
    constraintSetup [{transformType=Rotation, targetBone {boneName=thigh_l}}]
    constraintWeights [1]
   }
   n15 = "Constraint - dyn_fr_strap_base_r (R)"(ComponentPose=@n14, Alpha=1, ConstraintWeights_0=1) {
    alphaBoolBlend {blendOption=Linear}
    boneToModify {boneName=dyn_fr_strap_base_r}
    constraintSetup [{transformType=Rotation, targetBone {boneName=thigh_r}}]
    constraintWeights [1]
   }
   n16 = "Transform (Modify) Bone - Bone: dyn_fr_strap_base_l"(ComponentPose=@n15, Alpha=Get uplegRot_L(), Translation="1,1,0", Rotation="0,15,0.000", Scale="1,1,1") {
    rotationMode=BMM_Additive
    rotationSpace=BCS_BoneSpace
    translationMode=BMM_Additive
    translationSpace=BCS_BoneSpace
    alphaBoolBlend {blendOption=Linear}
    alphaScaleBiasClamp {bMapRange=true, inRange {max=0, min=-100}, outRange {max=0, min=1}}
    boneToModify {boneName=dyn_fr_strap_base_l}
    rotation {yaw=15}
    translation=<1 1>
   }
   n17 = "Transform (Modify) Bone - Bone: dyn_fr_strap_base_r"(ComponentPose=@n16, Alpha=Get uplegRot_R(), Translation="-1,-1,0", Rotation="0.000,15,0.000", Scale="1,1,1") {
    rotationMode=BMM_Additive
    rotationSpace=BCS_BoneSpace
    translationMode=BMM_Additive
    translationSpace=BCS_BoneSpace
    alphaBoolBlend {blendOption=Linear}
    alphaScaleBiasClamp {bMapRange=true, inRange {max=0, min=-100}, outRange {max=0, min=1}}
    boneToModify {boneName=dyn_fr_strap_base_r}
    rotation {yaw=15}
    translation=<-1 -1>
   }
   n18 = Get kneeRot_L() -> (kneeRot_L:double)
   n19 = "Transform (Modify) Bone - Bone: dyn_pants_fr_base_l"(ComponentPose=@n17, Alpha=@n18, Translation="-7,0.000,0.000", Rotation="0,0,0", Scale="1,1,1") {
    rotationMode=BMM_Additive
    rotationSpace=BCS_BoneSpace
    translationMode=BMM_Additive
    translationSpace=BCS_BoneSpace
    alphaBoolBlend {blendOption=Linear}
    alphaScaleBiasClamp {bMapRange=true, inRange {max=110, min=20}}
    boneToModify {boneName=dyn_pants_fr_base_l}
    translation {x=-7}
   }
   n20 = "Transform (Modify) Bone - Bone: dyn_pants_bk_base_l"(ComponentPose=@n19, Alpha=@n18, Translation="-20,0.000,0.000", Rotation="0,20,0", Scale="1,1,1") {
    rotationMode=BMM_Additive
    rotationSpace=BCS_BoneSpace
    translationMode=BMM_Additive
    translationSpace=BCS_BoneSpace
    alphaBoolBlend {blendOption=Linear}
    alphaScaleBiasClamp {bMapRange=true, inRange {max=110, min=20}}
    boneToModify {boneName=dyn_pants_bk_base_l}
    rotation {yaw=20}
    translation {x=-20}
   }
   n21 = Get kneeRot_R() -> (kneeRot_R:double)
   n22 = "Transform (Modify) Bone - Bone: dyn_pants_fr_base_r"(ComponentPose=@n20, Alpha=@n21, Translation="7,0.000,0", Rotation="0,0,0", Scale="1,1,1") {
    rotationMode=BMM_Additive
    rotationSpace=BCS_BoneSpace
    translationMode=BMM_Additive
    translationSpace=BCS_BoneSpace
    alphaBoolBlend {blendOption=Linear}
    alphaScaleBiasClamp {bMapRange=true, inRange {max=110, min=20}}
    boneToModify {boneName=dyn_pants_fr_base_r}
    translation {x=7}
   }
   n23 = "Transform (Modify) Bone - Bone: dyn_pants_bk_base_r"(ComponentPose=@n22, Alpha=@n21, Translation="20,0.000,0", Rotation="0,20,0", Scale="1,1,1") {
    rotationMode=BMM_Additive
    rotationSpace=BCS_BoneSpace
    translationMode=BMM_Additive
    translationSpace=BCS_BoneSpace
    alphaBoolBlend {blendOption=Linear}
    alphaScaleBiasClamp {bMapRange=true, inRange {max=110, min=20}}
    boneToModify {boneName=dyn_pants_bk_base_r}
    rotation {yaw=20}
    translation {x=20}
   }
   n24 = RigidBody(ComponentPose=@n23) {
    bClampLinearTranslationLimitToRefPose=true
    bOverrideWorldGravity=true
    overridePhysicsAsset=&/M_MED_Military_Fashion_Summer_Physics
    simulationSpace=BaseBoneSpace
    alphaBoolBlend {blendOption=Linear}
    baseBoneRef {boneName=pelvis}
    componentLinearAccScale=<0.4 0.4 0.4>
    componentLinearVelScale=<0.4 0.4 0.4>
    overrideWorldGravity {z=-980}
   }
   n25 = Get isCrouchingRev() -> (isCrouchingRev:double)
   n26 = RigidBody(ComponentPose=@n24, Alpha=@n25) {
    bClampLinearTranslationLimitToRefPose=true
    bOverrideWorldGravity=true
    overridePhysicsAsset=&/M_MED_Military_Fashion_Summer_String_Physics
    simulationSpace=BaseBoneSpace
    alphaBoolBlend {blendOption=Linear}
    baseBoneRef {boneName=root}
    componentLinearAccScale=<0.4 0.4 0.4>
    componentLinearVelScale=<0.4 0.4 0.4>
    overrideWorldGravity {z=-980}
   }
   n27 = RigidBody(ComponentPose=@n26, Alpha=@n25) {
    bClampLinearTranslationLimitToRefPose=true
    bOverrideWorldGravity=true
    overridePhysicsAsset=&/M_MED_Military_Fashion_Summer_Pants_Physics
    simulationSpace=BaseBoneSpace
    alphaBoolBlend {blendOption=Linear}
    alphaScaleBiasClamp {bMapRange=true, outRange {min=0.3}}
    baseBoneRef {boneName=root}
    componentLinearAccScale=<0.4 0.4 0.4>
    componentLinearVelScale=<0.4 0.4 0.4>
    overrideWorldGravity {z=-980}
   }
   n28 = Look At - Bone: dyn_pouch_bukle_strap_a_aim(ComponentPose=@n27, Alpha=1) {
    alphaBoolBlend {blendOption=Linear}
    boneToModify {boneName=dyn_pouch_bukle_strap_a_aim}
    lookAtLocation {x=0}
    lookAtTarget {boneReference {boneName=dyn_pouch_bukle_strap_a_tgt}}
    lookAt_Axis {axis=<1 0>}
    lookUp_Axis {axis {x=0, z=1}}
   }
   n29 = Look At - Bone: dyn_pouch_bukle_strap_b_aim(ComponentPose=@n28, Alpha=1) {
    alphaBoolBlend {blendOption=Linear}
    boneToModify {boneName=dyn_pouch_bukle_strap_b_aim}
    lookAtLocation {x=0}
    lookAtTarget {boneReference {boneName=dyn_pouch_bukle_strap_b_tgt}}
    lookAt_Axis {axis=<1 0>}
    lookUp_Axis {axis {x=0, z=1}}
   }
   n30 = "Transform (Modify) Bone - Bone: dyn_pouch_bukle_strap_a_aim"(ComponentPose=@n29, Alpha=Get uplegRot_R(), Translation="0,0,0", Rotation="40,0,0.000", Scale="1,1,1") {
    rotationMode=BMM_Additive
    rotationSpace=BCS_BoneSpace
    translationMode=BMM_Additive
    translationSpace=BCS_BoneSpace
    alphaBoolBlend {blendOption=Linear}
    alphaScaleBiasClamp {bMapRange=true, inRange {max=-70, min=-100}, outRange {max=0, min=1}}
    boneToModify {boneName=dyn_pouch_bukle_strap_a_aim}
    rotation {pitch=40}
   }
   Output Pose(Result=Component To Local(ComponentPose=@n30)) {layerGroup=None}
  ]}
 EventGraph {nodes [
   Event BlueprintUpdateAnimation()
   > n1 = Sequence()
   > n2 = Sequence!(execute) -> (then_0, then_1, then_2:Exec, then_3:Exec)
   n3 = GetDeltaTransformFromRefPose(self=GetOwnerSkeletalMeshComponent(), BoneName=thigh_r, BaseName=pelvis)
   Set uplegRot_R(execute=@n2.then_1, uplegRot_R:double=BreakRotator(InRot=BreakTransform(InTransform=@n3).Rotation).Yaw) -> (Output_Get:double)
   n4 = GetDeltaTransformFromRefPose(self=GetOwnerSkeletalMeshComponent(), BoneName=calf_r, BaseName=thigh_r)
   Set kneeRot_R(execute=@n2.then_3, kneeRot_R:double=BreakRotator(InRot=BreakTransform(InTransform=@n4).Rotation).Yaw) -> (Output_Get:double)
   n5 = GetDeltaTransformFromRefPose(self=GetOwnerSkeletalMeshComponent(), BoneName=calf_l, BaseName=thigh_l)
   Set kneeRot_L(execute=@n2.then_2, kneeRot_L:double=BreakRotator(InRot=BreakTransform(InTransform=@n5).Rotation).Yaw) -> (Output_Get:double)
   n6 = GetDeltaTransformFromRefPose(self=GetOwnerSkeletalMeshComponent(), BoneName=thigh_l, BaseName=pelvis)
   Set uplegRot_L(execute=@n2.then_0, uplegRot_L:double=BreakRotator(InRot=BreakTransform(InTransform=@n6).Rotation).Yaw) -> (Output_Get:double)
   n7 = Branch(execute=@n1.then_1, Condition=Get bIsCrouching())
   > Set isCrouchingRev(isCrouchingRev:double=0) -> (Output_Get:double)
   Set isCrouchingRev(execute=@n7.else, isCrouchingRev=1)
   TryGetPawnOwner()
  ]}
]
variables [
 isCrouchingRev {type=number, default=1}
 kneeRot_L {type=number}
 kneeRot_R {type=number}
 uplegRot_L {type=number}
 uplegRot_R {type=number}
]
