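# Physics asset for the Military Fashion Summer medium male player body, covering the dynamic hood bones and upper spine.
# A box represents the hood, a tiny capsule its base, and a larger capsule the spine_05 torso area.
# Constraints chain hood to base to spine: the base rotates freely against the spine, while the hood has asymmetric swing limits and a tight twist limit against the base.
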PhysicsAsset /BRCosmetics/Characters/Player/Male/Medium/Bodies/M_MED_Military_Fashion_Summer/Meshes/Parts/M_MED_Military_Fashion_Summer_Physics
bodies {
 dyn_hood [{shapeType=Box, extentX=13, extentY=5, extentZ=17, center=<6.0517 0 0>}]
 dyn_hood_base [{shapeType=Capsule, length=0.505, radius=0.505, rotation=<rot 0 -90 0>}]
 spine_05 [{shapeType=Capsule, length=13.5, radius=10.75, center=<0.8526 -0.6058 0>}]
}
constraints [
 {
  bone1Name=dyn_hood_base
  bone2Name=spine_05
  swing1LimitDegrees=45
  swing1Motion=Free
  swing2LimitDegrees=45
  swing2Motion=Free
  twistLimitDegrees=45
  twistMotion=Free
 }
 {bone1Name=dyn_hood, bone2Name=dyn_hood_base, swing1LimitDegrees=45, swing2LimitDegrees=15, twistLimitDegrees=5}
]
