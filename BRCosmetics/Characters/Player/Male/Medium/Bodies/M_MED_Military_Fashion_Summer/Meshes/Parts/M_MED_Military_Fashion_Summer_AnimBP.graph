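# Anim Blueprint for the Military Fashion Summer body part of a medium male player, adding secondary hood and hair motion.
# It layers copied poses at the upper arms, then runs RigidBody and spring-based AnimDynamics on hood and hair.
# Right-hood bones get corrective rotation driven by right upper-arm angles relative to the clavicle.
# While sprinting, hair bones blend in an additive roll mapped from head-mesh pelvis rotation.
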AnimBlueprint /BRCosmetics/Characters/Player/Male/Medium/Bodies/M_MED_Military_Fashion_Summer/Meshes/Parts/M_MED_Military_Fashion_Summer_AnimBP
target_skeleton=&/M_MED_Military_Fashion_Summer_Skeleton
graphs [
 AnimGraph {nodes [
   n1 = Copy Pose From Mesh(SourceMeshComponent=Get Body_Mesh())
   n2 = Copy Pose From Mesh(SourceMeshComponent=Get Head_Mesh() -> (Head_Mesh:Skeletal Mesh Component Object Reference)) {bUseAttachedParent=true}
   n3 = Get hoodFix_alpha_r() -> (hoodFix_alpha_r:double)
   n4 = "Transform (Modify) Bone - Bone: dyn_hood_bottom_base_r"(
    ComponentPose=Local To Component(LocalPose=Layered blend per bone(BasePose=@n2, BlendPoses_0=@n1) {layerSetup [{branchFilters [{boneName=upperarm_l}, {boneName=upperarm_r}]}]})
    Alpha=@n3
    Translation="0,0,0"
    Rotation="0,0,60"
    Scale="1,1,1"
   ) {
    rotationMode=BMM_Additive
    rotationSpace=BCS_BoneSpace
    translationMode=BMM_Additive
    translationSpace=BCS_BoneSpace
    alphaBoolBlend {blendOption=Linear}
    boneToModify {boneName=dyn_hood_bottom_base_r}
    rotation {roll=60}
   }
   n5 = "Transform (Modify) Bone - Bone: dyn_hood_base_r"(ComponentPose=@n4, Alpha=@n3, Translation="0,0,0", Rotation="0,0,25", Scale="1,1,1") {
    rotationMode=BMM_Additive
    rotationSpace=BCS_BoneSpace
    translationMode=BMM_Additive
    translationSpace=BCS_BoneSpace
    alphaBoolBlend {blendOption=Linear}
    boneToModify {boneName=dyn_hood_base_r}
    rotation {roll=25}
   }
   n6 = RigidBody(ComponentPose=@n5) {
    bClampLinearTranslationLimitToRefPose=true
    bOverrideWorldGravity=true
    overridePhysicsAsset=&/M_MED_Military_Fashion_Summer_Physics
    simulationSpace=BaseBoneSpace
    alphaBoolBlend {blendOption=Linear}
    baseBoneRef {boneName=spine_02}
    componentLinearAccScale=<0.4 0.4 0.4>
    componentLinearVelScale=<0.4 0.4 0.4>
    overrideWorldGravity {z=-980}
   }
   n7 = Anim Dynamics - Bone: dyn_hood_l(ComponentPose=@n6, Alpha=1, ExternalForce="0,0,0") {
    angularDampingOverride=1
    angularSpringConstant=1500
    bAngularSpring=true
    bGravityOverrideInSimSpace=true
    bOverrideAngularDamping=true
    gravityScale=0.35
    simulationSpace=RootRelative
    alphaBoolBlend {blendOption=Linear}
    boundBone {boneName=dyn_hood_l}
    physicsBodyDefinitions [{
      sphereCollisionRadius=0
      boundBone {boneName=dyn_hood_l}
      boxExtents=<10 20 20>
      constraintSetup {angularLimitsMax=<15 5 5>, angularLimitsMin=<-15 -5 -5>, angularTarget=<0 0 0>}
      localJointOffset=<10 0 0>
     }]
   }
   n8 = Anim Dynamics - Bone: dyn_hood_r(ComponentPose=@n7, Alpha=1, ExternalForce="0,0,0") {
    angularDampingOverride=1
    angularSpringConstant=1500
    bAngularSpring=true
    bGravityOverrideInSimSpace=true
    bOverrideAngularDamping=true
    gravityScale=0.35
    simulationSpace=RootRelative
    alphaBoolBlend {blendOption=Linear}
    boundBone {boneName=dyn_hood_r}
    physicsBodyDefinitions [{
      sphereCollisionRadius=0
      boundBone {boneName=dyn_hood_r}
      boxExtents=<10 20 20>
      constraintSetup {angularLimitsMax=<15 5 5>, angularLimitsMin=<-15 -5 -5>, angularTarget=<0 0 0>}
      localJointOffset=<10 0 0>
     }]
   }
   n9 = Anim Dynamics - Bone: dyn_hair_d(ComponentPose=@n8, Alpha=1, ExternalForce="0,0,0") {
    angularBiasOverride=0.9
    angularDampingOverride=0.8
    angularSpringConstant=100
    bAngularSpring=true
    bOverrideAngularDamping=true
    bUseGravityOverride=true
    alphaBoolBlend {blendOption=Linear}
    boundBone {boneName=dyn_hair_d}
    componentAppliedLinearAccClamp=<3000 3000 3000>
    componentLinearAccScale=<0.8 0.8 0.8>
    componentLinearVelScale=<0.8 0.8 0.8>
    gravityOverride {z=-980}
    physicsBodyDefinitions [{
      boundBone {boneName=dyn_hair_d}
      boxExtents=<5 5 5>
      constraintSetup {angularLimitsMax=<30 10 8>, angularLimitsMin=<-30 -10 -60>, angularTarget=<1 0 0>}
      localJointOffset=<0.2 0 0>
     }]
   }
   n10 = Anim Dynamics - Bone: dyn_hair_e(ComponentPose=@n9, Alpha=1, ExternalForce="0,0,0") {
    angularBiasOverride=0.9
    angularDampingOverride=0.8
    angularSpringConstant=100
    bAngularSpring=true
    bOverrideAngularDamping=true
    bUseGravityOverride=true
    alphaBoolBlend {blendOption=Linear}
    boundBone {boneName=dyn_hair_e}
    componentAppliedLinearAccClamp=<3000 3000 3000>
    componentLinearAccScale=<0.8 0.8 0.8>
    componentLinearVelScale=<0.8 0.8 0.8>
    gravityOverride {z=-980}
    physicsBodyDefinitions [{
      boundBone {boneName=dyn_hair_e}
      boxExtents=<5.5 5.5 5.5>
      constraintSetup {angularLimitsMax=<30 10 8>, angularLimitsMin=<-30 -10 -60>, angularTarget=<1 0 0>}
      localJointOffset=<0.2 0 0>
     }]
   }
   n11 = Anim Dynamics - Bone: dyn_hair_f(ComponentPose=@n10, Alpha=1, ExternalForce="0,0,0") {
    angularBiasOverride=0.9
    angularDampingOverride=0.8
    angularSpringConstant=100
    bAngularSpring=true
    bOverrideAngularDamping=true
    bUseGravityOverride=true
    alphaBoolBlend {blendOption=Linear}
    boundBone {boneName=dyn_hair_f}
    componentAppliedLinearAccClamp=<3000 3000 3000>
    componentLinearAccScale=<0.8 0.8 0.8>
    componentLinearVelScale=<0.8 0.8 0.8>
    gravityOverride {z=-980}
    physicsBodyDefinitions [{
      boundBone {boneName=dyn_hair_f}
      boxExtents=<4 4 4>
      constraintSetup {angularLimitsMax=<30 10 8>, angularLimitsMin=<-30 -10 -60>, angularTarget=<1 0 0>}
      localJointOffset=<0.2 0 0>
     }]
   }
   n12 = Anim Dynamics - Bone: dyn_hair_g(ComponentPose=@n11, Alpha=1, ExternalForce="0,0,0") {
    angularBiasOverride=0.9
    angularDampingOverride=0.8
    angularSpringConstant=100
    bAngularSpring=true
    bOverrideAngularDamping=true
    bUseGravityOverride=true
    alphaBoolBlend {blendOption=Linear}
    boundBone {boneName=dyn_hair_g}
    componentAppliedLinearAccClamp=<3000 3000 3000>
    componentLinearAccScale=<0.8 0.8 0.8>
    componentLinearVelScale=<0.8 0.8 0.8>
    gravityOverride {z=-980}
    physicsBodyDefinitions [{
      boundBone {boneName=dyn_hair_g}
      boxExtents=<5 5 5>
      constraintSetup {angularLimitsMax=<30 10 8>, angularLimitsMin=<-30 -10 -60>, angularTarget=<1 0 0>}
      localJointOffset=<0.2 0 0>
     }]
   }
   n13 = Anim Dynamics - Bone: dyn_hair_c(ComponentPose=@n12, Alpha=1, ExternalForce="0,0,0") {
    angularBiasOverride=0.9
    angularDampingOverride=0.8
    angularSpringConstant=100
    bAngularSpring=true
    bOverrideAngularDamping=true
    bUseGravityOverride=true
    alphaBoolBlend {blendOption=Linear}
    boundBone {boneName=dyn_hair_c}
    componentAppliedLinearAccClamp=<3000 3000 3000>
    componentLinearAccScale=<0.8 0.8 0.8>
    componentLinearVelScale=<0.8 0.8 0.8>
    gravityOverride {z=-980}
    physicsBodyDefinitions [{
      boundBone {boneName=dyn_hair_c}
      boxExtents=<4.5 4.5 4.5>
      constraintSetup {angularLimitsMax=<30 10 8>, angularLimitsMin=<-30 -10 -60>, angularTarget=<1 0 0>}
      localJointOffset=<0.2 0 0>
     }]
   }
   n14 = Anim Dynamics - Bone: dyn_hair_b(ComponentPose=@n13, Alpha=1, ExternalForce="0,0,0") {
    angularBiasOverride=0.9
    angularDampingOverride=0.8
    angularSpringConstant=100
    bAngularSpring=true
    bOverrideAngularDamping=true
    bUseGravityOverride=true
    alphaBoolBlend {blendOption=Linear}
    boundBone {boneName=dyn_hair_b}
    componentAppliedLinearAccClamp=<3000 3000 3000>
    componentLinearAccScale=<0.8 0.8 0.8>
    componentLinearVelScale=<0.8 0.8 0.8>
    gravityOverride {z=-980}
    physicsBodyDefinitions [{
      boundBone {boneName=dyn_hair_b}
      boxExtents=<5.2 5 5>
      constraintSetup {angularLimitsMax=<30 10 8>, angularLimitsMin=<-30 -10 -60>, angularTarget=<1 0 0>}
      localJointOffset=<0.2 0 0>
     }]
   }
   n15 = Anim Dynamics - Bone: dyn_hair_a(ComponentPose=@n14, Alpha=1, ExternalForce="0,0,0") {
    angularBiasOverride=0.9
    angularDampingOverride=0.8
    angularSpringConstant=100
    bAngularSpring=true
    bOverrideAngularDamping=true
    bUseGravityOverride=true
    alphaBoolBlend {blendOption=Linear}
    boundBone {boneName=dyn_hair_a}
    componentAppliedLinearAccClamp=<3000 3000 3000>
    componentLinearAccScale=<0.8 0.8 0.8>
    componentLinearVelScale=<0.8 0.8 0.8>
    gravityOverride {z=-980}
    physicsBodyDefinitions [{
      boundBone {boneName=dyn_hair_a}
      boxExtents=<5 5.2 5.2>
      constraintSetup {angularLimitsMax=<30 10 8>, angularLimitsMin=<-30 -10 -60>, angularTarget=<1 0 0>}
      localJointOffset=<0.2 0 0>
     }]
   }
   n16 = Anim Dynamics - Bone: dyn_hair_h(ComponentPose=@n15, Alpha=1, ExternalForce="0,0,0") {
    angularBiasOverride=0.9
    angularDampingOverride=0.8
    angularSpringConstant=100
    bAngularSpring=true
    bOverrideAngularDamping=true
    bUseGravityOverride=true
    alphaBoolBlend {blendOption=Linear}
    boundBone {boneName=dyn_hair_h}
    componentAppliedLinearAccClamp=<3000 3000 3000>
    componentLinearAccScale=<0.8 0.8 0.8>
    componentLinearVelScale=<0.8 0.8 0.8>
    gravityOverride {z=-980}
    physicsBodyDefinitions [{
      boundBone {boneName=dyn_hair_h}
      boxExtents=<4 4 4>
      constraintSetup {angularLimitsMax=<30 10 8>, angularLimitsMin=<-30 -10 -60>, angularTarget=<1 0 0>}
      localJointOffset=<0.2 0 0>
     }]
   }
   n17 = Get hair_rotation() -> (hair_rotation:Rotator)
   n18 = Get bIsSprinting()
   n19 = "Transform (Modify) Bone - Bone: dyn_hair_d"!(ComponentPose=@n16, bAlphaBoolEnabled:bool=@n18, Translation="0,0,0", Rotation=@n17, Scale="1,1,1") -> (Pose) {
    alphaInputType=Bool
    rotationMode=BMM_Additive
    rotationSpace=BCS_BoneSpace
    alphaBoolBlend {blendInTime=0.3, blendOption=Linear, blendOutTime=1}
    boneToModify {boneName=dyn_hair_d}
   }
   n20 = "Transform (Modify) Bone - Bone: dyn_hair_e"!(ComponentPose=@n19, bAlphaBoolEnabled:bool=@n18, Translation="0,0,0", Rotation=@n17, Scale="1,1,1") -> (Pose) {
    alphaInputType=Bool
    rotationMode=BMM_Additive
    rotationSpace=BCS_BoneSpace
    alphaBoolBlend {blendInTime=0.3, blendOption=Linear, blendOutTime=1}
    boneToModify {boneName=dyn_hair_e}
   }
   n21 = "Transform (Modify) Bone - Bone: dyn_hair_f"!(ComponentPose=@n20, bAlphaBoolEnabled:bool=@n18, Translation="0,0,0", Rotation=@n17, Scale="1,1,1") -> (Pose) {
    alphaInputType=Bool
    rotationMode=BMM_Additive
    rotationSpace=BCS_BoneSpace
    alphaBoolBlend {blendInTime=0.3, blendOption=Linear, blendOutTime=1}
    boneToModify {boneName=dyn_hair_f}
   }
   n22 = "Transform (Modify) Bone - Bone: dyn_hair_g"!(ComponentPose=@n21, bAlphaBoolEnabled:bool=@n18, Translation="0,0,0", Rotation=@n17, Scale="1,1,1") -> (Pose) {
    alphaInputType=Bool
    rotationMode=BMM_Additive
    rotationSpace=BCS_BoneSpace
    alphaBoolBlend {blendInTime=0.3, blendOption=Linear, blendOutTime=1}
    boneToModify {boneName=dyn_hair_g}
   }
   n23 = "Transform (Modify) Bone - Bone: dyn_hair_c"!(ComponentPose=@n22, bAlphaBoolEnabled:bool=@n18, Translation="0,0,0", Rotation=@n17, Scale="1,1,1") -> (Pose) {
    alphaInputType=Bool
    rotationMode=BMM_Additive
    rotationSpace=BCS_BoneSpace
    alphaBoolBlend {blendInTime=0.3, blendOption=Linear, blendOutTime=1}
    boneToModify {boneName=dyn_hair_c}
   }
   n24 = "Transform (Modify) Bone - Bone: dyn_hair_b"!(ComponentPose=@n23, bAlphaBoolEnabled:bool=@n18, Translation="0,0,0", Rotation=@n17, Scale="1,1,1") -> (Pose) {
    alphaInputType=Bool
    rotationMode=BMM_Additive
    rotationSpace=BCS_BoneSpace
    alphaBoolBlend {blendInTime=0.3, blendOption=Linear, blendOutTime=1}
    boneToModify {boneName=dyn_hair_b}
   }
   n25 = "Transform (Modify) Bone - Bone: dyn_hair_a"!(ComponentPose=@n24, bAlphaBoolEnabled:bool=@n18, Translation="0,0,0", Rotation=@n17, Scale="1,1,1") -> (Pose) {
    alphaInputType=Bool
    rotationMode=BMM_Additive
    rotationSpace=BCS_BoneSpace
    alphaBoolBlend {blendInTime=0.3, blendOption=Linear, blendOutTime=1}
    boneToModify {boneName=dyn_hair_a}
   }
   n26 = "Transform (Modify) Bone - Bone: dyn_hair_h"!(ComponentPose=@n25, bAlphaBoolEnabled:bool=@n18, Translation="0,0,0", Rotation=@n17, Scale="1,1,1") -> (Pose) {
    alphaInputType=Bool
    rotationMode=BMM_Additive
    rotationSpace=BCS_BoneSpace
    alphaBoolBlend {blendInTime=0.3, blendOption=Linear, blendOutTime=1}
    boneToModify {boneName=dyn_hair_h}
   }
   Output Pose(Result=Component To Local(ComponentPose=@n26)) {layerGroup=None}
  ]}
 EventGraph {nodes [
   Event BlueprintUpdateAnimation()
   > n1 = Sequence!(execute) -> (then_0, then_1, then_2:Exec, then_3:Exec)
   n2 = GetOwnerSkeletalMeshComponent()
   n3 = BreakRotator(InRot=BreakTransform(InTransform=GetDeltaTransformFromRefPose(self=@n2, BoneName=upperarm_r, BaseName=clavicle_r)).Rotation)
   > Set armForward_R(armForward_R:double=@n3.Yaw) -> (Output_Get:double)
   > Set armUp_R(armUp_R:double=@n3.Pitch) -> (Output_Get:double)
   n4 = GetSocketTransform(self=Get Head_Mesh(), InSocketName=pelvis, TransformSpace=RTS_Component)
   n5 = "Absolute (Float)"(A=BreakRotator(InRot=BreakTransform(InTransform=@n4).Rotation))
   > Set hair_rotation(hair_rotation:Rotator=MakeRotator(Roll=MapRangeClamped(Value=@n5, InRangeA=80, InRangeB=100, OutRangeA=-30, OutRangeB=30))) -> (Output_Get:Rotator)
   n6 = Branch(execute=@n1.then_1, Condition="float > float"(A=Get armForward_R() -> (armForward_R:double), B=90) {is=title K2Node_CallFunction})
   > Set hoodFix_alpha_r(hoodFix_alpha_r:double=MapRangeClamped(Value=Get armUp_R() -> (armUp_R:double), InRangeA=-10, InRangeB=10)) -> (Output_Get:double)
   Set hoodFix_alpha_r(execute=@n6.else, hoodFix_alpha_r=0)
   Event BlueprintInitializeAnimation()
   > Set Body_Mesh(Body_Mesh=GetOwnerSkeletalMeshComponent())
   > n7 = Cast To PlayerPawn_Athena(Object=TryGetPawnOwner())
   > n8 = Set Pawn Owner(Pawn Owner:Player Pawn Athena Object Reference=@n7) -> (Output_Get:Player Pawn Athena Object Reference)
   > Set Head_Mesh(Head_Mesh:Skeletal Mesh Component Object Reference=GetSkeletalMeshForPartType(self=@n8, PartType=Head)) -> (Output_Get:Skeletal Mesh Component Object Reference)
  ]}
]
variables [
 armForward_R {type=number}
 armUp_R {type=number}
 Body_Mesh {type=SkeletalMeshComponent}
 hair_rotation {type=Rotator}
 Head_Mesh {type=SkeletalMeshComponent}
 hoodFix_alpha_r {type=number}
 Pawn Owner {type=PlayerPawn_Athena_C}
]
